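; Actor component blueprint that stores the charge percent for the Cove Spice weapon.
; A custom event, Update Percent, takes a new float and writes it straight into the Cove Spice Charge Percent variable.
; The file shows no other logic, clamping or notification.
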(event Custom|Update%20Percent
  (params (New%20Percent :type Float%20%28double-precision%29))
  (Variables|Default|SetCove%20Spice%20Charge%20Percent New%20Percent))
