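; Anim notify state that swaps materials on the CoveSpice pistol mesh for the duration of its equip animation.
; On begin, it replaces the hardcoded overlay with a blank-hull material and applies a dynamic instance of the override material to slot 0.
; That instance receives the current game time as BP_EquipStartTime, so the shader drives the blend itself without per-tick updates.
; On end, it restores the inverted-hull overlay and the default material.
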(fn Received_NotifyBegin
  (params (MeshComp :type Skeletal%20Mesh%20Component%20Object%20Reference) (Animation :type Anim%20Sequence%20Base%20Object%20Reference) (TotalDuration :type Float%20%28single-precision%29) (EventReference :type Anim%20Notify%20Event%20Reference%20Structure%20%28by%20ref%29))
  (bind _meshcomp (Variables|GetMeshComp))
  ; We need to remove the Hardcoded Overlay
  (Rendering|Material|SetOverlayMaterial MeshComp "/WildEstateWeapons/Assets/CoveSpice_Pistol/Materials/MI_CoveSpice_Pistol_BlankHull_EquipAnim.MI_CoveSpice_Pistol_BlankHull_EquipAnim" false 0 :!class /Script/Engine.MeshComponent)
  ; Set new material to out instance - Set the time we are at while transitioning in the material to use for the blend
  (bind _returnvalue (Rendering|Material|CreateDynamicMaterialInstance _meshcomp 0 (Variables|Default|GetOverrideMaterial) "None" :!class /Script/Engine.PrimitiveComponent))
  (Rendering|Material|SetScalarParameterValue _returnvalue "BP_EquipStartTime" (Utilities|Time|GetTimeSeconds _meshcomp) :!class /Script/Engine.MaterialInstanceDynamic)
  (Rendering|Material|SetMaterial (Variables|GetMeshComp) 0 _returnvalue)
  (return true))

(fn Received_NotifyEnd
  (params (MeshComp :type Skeletal%20Mesh%20Component%20Object%20Reference) (Animation :type Anim%20Sequence%20Base%20Object%20Reference) (EventReference :type Anim%20Notify%20Event%20Reference%20Structure%20%28by%20ref%29))
  ; Reset the material
  (Rendering|Material|SetOverlayMaterial MeshComp "/WildEstateWeapons/Assets/CoveSpice_Pistol/Materials/MI_CoveSpice_Pistol_InvHull.MI_CoveSpice_Pistol_InvHull" false 0 :!class /Script/Engine.MeshComponent)
  (Rendering|Material|SetMaterial MeshComp 0 (Variables|Default|GetDefaultMaterial))
  (return true))
