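# Animation Blueprint for the CoveSpice arm gadget skeleton, posing the arm from a tentacle equip sequence.
# The sequence feeds DefaultSlot for montage playback, then a root bone modifier zeroes translation and adds a 90-degree pitch.
# The offset bone then gets a world-space rotation with yaw from a property access and an additive Z shift from TranslationZ.
# The event graph is empty and the Blueprint declares no variables of its own.
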AnimBlueprint /WildEstateWeapons/Gameplay/CoveSpice/Animation/CoveSpice_ArmGadget_AnimBP
target_skeleton=/WildEstateWeapons/Assets/CoveSpice_Pistol/Arm/Meshes/CoveSpice_Arm_Gadget_Skeleton
graphs [
 AnimGraph {nodes [
   n1 = Sequence Player 'CoveSpice_Equip_Tentacle_F1'() {sequence=&/CoveSpice_Equip_Tentacle_F1}
   n2 = "Transform (Modify) Bone - Bone: root"(ComponentPose=Local To Component(LocalPose=Slot 'DefaultSlot'(Source=@n1)), Translation="0,0,0", Rotation="90,0,0") {rotationMode=BMM_Additive, scaleMode=BMM_Replace, translationMode=BMM_Replace, boneToModify {boneName=root}, rotation {pitch=90}}
   n3 = Property Access!() -> (Value:double=~) {is=title K2Node_PropertyAccess}
   n4 = Get TranslationZ() -> (TranslationZ:double)
   Output Pose(Result=Component To Local(ComponentPose="Transform (Modify) Bone - Bone: offset"!(
      ComponentPose=@n2
      Alpha=1
      Translation_X:double=0
      Translation_Y:double=0
      Translation_Z:double=@n4
      Rotation_Roll:float=0
      Rotation_Pitch:float=0
      Rotation_Yaw:float=@n3
      Scale="1,1,1"
     ) -> (Pose) {rotationMode=BMM_Replace, rotationSpace=BCS_WorldSpace, translationMode=BMM_Additive, boneToModify {boneName=offset}}))
  ]}
 EventGraph {links [], nodes []}
]
variables []
