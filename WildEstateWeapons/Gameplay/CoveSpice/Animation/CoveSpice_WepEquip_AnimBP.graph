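# Animation Blueprint for the CoveSpice pistol skeleton that plays the weapon's equip animation and aligns its root bone with the player's stance.
# A non-looping equip sequence feeds DefaultSlot, then a Modify Bone node replaces root rotation in world space, using a property-accessed yaw minus 90 degrees.
# Each update casts the Fortnite player pawn and checks crouch state to set an additive root Z offset, lowering the weapon when crouched.
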AnimBlueprint /WildEstateWeapons/Gameplay/CoveSpice/Animation/CoveSpice_WepEquip_AnimBP
target_skeleton=/WildEstateWeapons/Assets/CoveSpice_Pistol/Meshes/CoveSpice_Pistol_Skeleton
graphs [
 AnimGraph {nodes [
   Input Pose()
   n1 = Local To Component(LocalPose=Slot 'DefaultSlot'(Source=Sequence Player 'CoveSpice_Equip_WEP'() {bLoopAnimation=false, sequence=&/CoveSpice_Equip_WEP}))
   n2 = float - float(A=Property Access!() -> (Value:double=~) {is=title K2Node_PropertyAccess}, B=90)
   n3 = Get Translation Z() -> (Translation Z:double)
   Output Pose(Result=Component To Local(ComponentPose="Transform (Modify) Bone - Bone: root"!(
      ComponentPose=@n1
      Alpha
      Translation_X:double=0
      Translation_Y:double=0
      Translation_Z:double=@n3
      Rotation_Roll:float=0
      Rotation_Pitch:float=0
      Rotation_Yaw:float=@n2
      Scale
     ) -> (Pose) {rotationMode=BMM_Replace, rotationSpace=BCS_WorldSpace, translationMode=BMM_Additive, boneToModify {boneName=root}}))
  ]}
 EventGraph {nodes [
   Event BlueprintUpdateAnimation()
   > n1 = Cast To FortPlayerPawn(Object=TryGetFortPawn())
   > n2 = Set As Fort Player Pawn(As Fort Player Pawn:Fort Player Pawn Object Reference=@n1) -> (Output_Get:Fort Player Pawn Object Reference)
   > n3 = Branch(Condition=Get bIsCrouched!(self:Character Object Reference=@n2) -> (bIsCrouched:bool))
   > Set Translation Z(Translation Z:double=-40) -> (Output_Get:double)
   Set Translation Z(execute=@n3.else, Translation Z=0)
   TryGetPawnOwner()
  ]}
]
variables [
 As Fort Player Pawn {type=FortPlayerPawn}
 Translation Z {type=number}
]
