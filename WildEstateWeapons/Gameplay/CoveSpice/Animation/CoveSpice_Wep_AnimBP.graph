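# Animation Blueprint for the CoveSpice pistol weapon skeleton, supplying the weapon mesh's base pose.
# A non-looping CoveSpice_Wep_Pose sequence player feeds DefaultSlot, so montages played on that slot can override the pose before output.
# The event graph holds only an unconnected TryGetPawnOwner node, and the IsFiring boolean is declared but not used in either graph.
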AnimBlueprint /WildEstateWeapons/Gameplay/CoveSpice/Animation/CoveSpice_Wep_AnimBP
target_skeleton=/WildEstateWeapons/Assets/CoveSpice_Pistol/Meshes/CoveSpice_Pistol_Skeleton
graphs [
 AnimGraph {nodes [Output Pose(Result=Slot 'DefaultSlot'(Source=Sequence Player 'CoveSpice_Wep_Pose'() {bLoopAnimation=false, sequence=&/CoveSpice_Wep_Pose})) {layerGroup=None}]}
 EventGraph {nodes [TryGetPawnOwner()]}
]
variables [IsFiring {type=boolean}]
