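; A function override for a fort spawn prop that customizes each prop mesh component's custom depth rendering.
; It first calls the parent CustomizePropMeshComponent, then sets render custom depth on the passed mesh component.
; The on or off state comes from the GetCustomDepthEnabled variable rather than a fixed value.
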(fn CustomizePropMeshComponent
  (params (MeshComponent :type Mesh%20Component%20Object%20Reference))
  (|CallParentFunction MeshComponent)
  (Rendering|SetRenderCustomDepth MeshComponent (Variables|Default|GetCustomDepthEnabled)))
