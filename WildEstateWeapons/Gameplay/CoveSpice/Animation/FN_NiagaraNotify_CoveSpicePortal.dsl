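; Received_Notify override for a Niagara anim notify, named for the CoveSpice portal, that can control custom depth rendering on the effect it spawns.
; It calls the parent notify, then casts the spawned effect to a Niagara component.
; If the custom depth override is on, it applies the configured enabled state to that component.
; It returns false whether or not the cast succeeds.
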(fn Received_Notify
  (params (MeshComp :type Skeletal%20Mesh%20Component%20Object%20Reference) (Animation :type Anim%20Sequence%20Base%20Object%20Reference) (EventReference :type Anim%20Notify%20Event%20Reference%20Structure%20%28by%20ref%29))
  (|CallParentFunction MeshComp Animation EventReference)
  (bind _returnvalue (AnimNotify|GetSpawnedEffect))
  (bind _asniagara_particle_system_component (Utilities|Casting|CastToNiagaraComponent _returnvalue)
    (:then
      (LocalVariables|SetNiagaraSystem _asniagara_particle_system_component)
      (if (Variables|Default|GetOverrideCustomDepth)
        (Rendering|SetRenderCustomDepth (LocalVariables|GetNiagaraSystem) (Variables|Default|GetCustomDepthEnabled)))
      (return false))
    (:CastFailed
      (return false))))
