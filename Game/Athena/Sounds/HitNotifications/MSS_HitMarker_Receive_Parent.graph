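# Parent one-shot stereo MetaSound for received-hit notifications, layering a random body sound with an optional, slightly delayed random UI tick by default.
# One random pitch shift per play is shared by both wave players, whose mixed output feeds an orientation-driven azimuth panner.
# An empty Layer 2 array skips the delayed layer.
# OnFinished then follows Layer 1 alone; otherwise a trigger accumulator waits for both players to finish.
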MetaSoundSource /Game/Athena/Sounds/HitNotifications/MSS_HitMarker_Receive_Parent
inputs [
 Layer 1 Sounds {type=WaveAsset:Array}
 Layer 2 Sounds {type=WaveAsset:Array}
 Layer02_DelayTime {type=Time}
 Max Pitch {type=Float}
 Min Pitch {type=Float}
 UE.Source.OnPlay {type=Trigger}
 UE.Source.Orientation.Azimuth {type=Float}
 UE.Source.Orientation.Elevation {type=Float}
]
interfaces [UE.OutputFormat.Stereo:1.0, UE.Source.OneShot:1.0, UE.Source.Orientation:1.0, UE.Source:1.1]
outputs [
 "UE.OutputFormat.Stereo.Audio:0" {type=Audio}
 "UE.OutputFormat.Stereo.Audio:1" {type=Audio}
 UE.Source.OneShot.OnFinished {type=Trigger}
]
pages [Default {
  input_defaults {
   Layer02_DelayTime=0.05
   Max Pitch=1.1
   Min Pitch=0.9
   UE.Source.OnPlay=null
   UE.Source.Orientation.Azimuth=0
   UE.Source.Orientation.Elevation=0
   Layer 1 Sounds [
    &/HitMarker_Body_01
    &/HitMarker_Body_02
    &/HitMarker_Body_03
   ]
   Layer 2 Sounds [
    &/HitMarker_Receive_UI_Tick_01
    &/HitMarker_Receive_UI_Tick_02
    &/HitMarker_Receive_UI_Tick_03
   ]
  }
  nodes [
   n1 = UE.RandomFloat(Max=^Max Pitch, Min=^Min Pitch, Next=^UE.Source.OnPlay)
   n2 = TriggerCompare.Clamp:Int32(A=Array.Num:WaveAsset:Array(Array=^Layer 2 Sounds).Num, Compare=@n1.On Next) {also {defaults {B=0}}}
   TriggerAnyTriggerAny2_1 = "TriggerAny.Trigger Any (2)"(In 1=@n2.False, In 0=@n2.True)
   ArrayRandomGet_1 = Array.Random Get:WaveAsset:Array(In Array=^Layer 1 Sounds, Next=@TriggerAnyTriggerAny2_1.Out)
   n3 = VariableAccessor.Float(Variable=VariableMutator.Float(Variable=InitVariable.Float().Variable, Value=@n1.Value).Variable)
   WavePlayer_0 = UE.Wave Player:Stereo(Play=@ArrayRandomGet_1.On Next, Wave Asset=@ArrayRandomGet_1.Value, Pitch Shift=@n3.Value) {also {version [1,1]}}
   TriggerDelay_0 = UE.Trigger Delay(Delay Time=^Layer02_DelayTime, In=@n2.False)
   ArrayRandomGet_0 = Array.Random Get:WaveAsset:Array(In Array=^Layer 2 Sounds, Next=@TriggerDelay_0.Out)
   WavePlayer_1 = UE.Wave Player:Stereo(Play=@ArrayRandomGet_0.On Next, Wave Asset=@ArrayRandomGet_0.Value, Pitch Shift=VariableAccessor.Float(Variable=@n3.Variable).Value) {also {version [1,1]}}
   AudioMixerAudioMixerStereo2_0 = "AudioMixer.Audio Mixer (Stereo, 2)"(In 0 L=@WavePlayer_0.Out Left, In 0 R=@WavePlayer_0.Out Right, In 1 L=@WavePlayer_1.Out Left, In 1 R=@WavePlayer_1.Out Right)
   AzimuthPannerStereo_0 = &AzimuthPannerStereo(Azimuth=^UE.Source.Orientation.Azimuth, In L=@AudioMixerAudioMixerStereo2_0.Out L, In R=@AudioMixerAudioMixerStereo2_0.Out R) {also {defaults {StereoSpread=850}}}
   InitVariable.Bool()
   n4 = TriggerCompare.Clamp:Int32(A=Array.Num:WaveAsset:Array(Array=^Layer 2 Sounds).Num, Compare=@WavePlayer_0.On Finished) {also {defaults {B=0}}}
   TriggerAccumulatorTriggerAccumulate3_0 = "TriggerAccumulator.Trigger Accumulate (3)"(In 2=UE.Reroute(Value=@WavePlayer_1.On Finished).Value, In 1=UE.Reroute(Value=@WavePlayer_0.On Finished).Value, In 0=@n4.False)
   TriggerAnyTriggerAny2_0 = "TriggerAny.Trigger Any (2)"(In 1=@TriggerAccumulatorTriggerAccumulate3_0.Out, In 0=@n4.True)
  ]
  links [
   {dst="@UE.OutputFormat.Stereo.Audio:0", src=AzimuthPannerStereo_0.Out L}
   {dst="@UE.OutputFormat.Stereo.Audio:1", src=AzimuthPannerStereo_0.Out R}
   {dst="@UE.Source.OneShot.OnFinished", src=TriggerAnyTriggerAny2_0.Out}
  ]
 }]
properties {
 assetImportData=&self:AssetImportData
 attenuationSettings=&/Att_HitMarker_Recieve_MS
 numChannels=2
 outputFormat=Stereo
 sampleRate=48000
 soundClassObject=/Game/Sounds/FortSoundClassesAndModes/Fort_2DHitNotifications_Receive_PSM
 concurrencySet [&HitNotifications_TwoMaximum_StopOldest_VolumeScale]
}
