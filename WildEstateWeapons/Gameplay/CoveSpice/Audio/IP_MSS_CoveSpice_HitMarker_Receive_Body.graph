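# CoveSpice weapon MetaSound preset for the body-hit "receive" hit marker, a stereo one-shot built on MSS_HitMarker_Receive_Parent.
# It feeds the parent shared Athena body variations as layer one and UI tick variations as a slightly delayed layer two, with a narrow pitch range.
# Source settings route it to the 2D receive notification sound class with hit-marker attenuation.
# An enabled concurrency override allows one instance, prevents new ones and limits retriggers.
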MetaSoundSource /WildEstateWeapons/Gameplay/CoveSpice/Audio/IP_MSS_CoveSpice_HitMarker_Receive_Body
preset_parent=/Game/Athena/Sounds/HitNotifications/MSS_HitMarker_Receive_Parent
inputs [
 Layer 1 Sounds {type=WaveAsset:Array}
 Layer 2 Sounds {type=WaveAsset:Array}
 Layer02_DelayTime {type=Time}
 Max Pitch {type=Float}
 Min Pitch {type=Float}
 UE.Source.OnPlay {type=Trigger}
 UE.Source.Orientation.Azimuth {type=Float}
 UE.Source.Orientation.Elevation {type=Float}
]
interfaces [UE.OutputFormat.Stereo:1.0, UE.Source.OneShot:1.0, UE.Source.Orientation:1.0, UE.Source:1.1]
outputs [
 "UE.OutputFormat.Stereo.Audio:0" {type=Audio}
 "UE.OutputFormat.Stereo.Audio:1" {type=Audio}
 UE.Source.OneShot.OnFinished {type=Trigger}
]
pages [Default {input_defaults {
   Layer02_DelayTime=0.05
   Max Pitch=1.1
   Min Pitch=0.9
   UE.Source.OnPlay=null
   UE.Source.Orientation.Azimuth=0
   UE.Source.Orientation.Elevation=0
   Layer 1 Sounds [
    /Game/Athena/Sounds/HitNotifications/HitMarker_Body_01
    /Game/Athena/Sounds/HitNotifications/HitMarker_Body_02
    /Game/Athena/Sounds/HitNotifications/HitMarker_Body_03
   ]
   Layer 2 Sounds [
    /Game/Athena/Sounds/HitNotifications/HitMarker_Receive_UI_Tick_01
    /Game/Athena/Sounds/HitNotifications/HitMarker_Receive_UI_Tick_02
    /Game/Athena/Sounds/HitNotifications/HitMarker_Receive_UI_Tick_03
   ]
  }}]
properties {
 assetImportData=&self:AssetImportData
 attenuationSettings=/Game/Athena/Sounds/HitNotifications/Att_HitMarker_Recieve_MS
 bOverrideConcurrency=true
 numChannels=2
 outputFormat=Stereo
 sampleRate=48000
 soundClassObject=/Game/Sounds/FortSoundClassesAndModes/Fort_2DHitNotifications_Receive_PSM
 volume=0.5
 concurrencyOverrides {maxCount=1, resolutionRule=PreventNew, retriggerTime=0.3}
 concurrencySet [&HitNotifications_TwoMaximum_StopOldest_VolumeScale]
}
