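# Looping mono MetaSound source for CoveSpice weapon impacts, with a sound that follows the surface being hit.
# A surface loop swapper preset crossfades to the supplied wave on each surface change, applies pitch shift, and can fade everything out.
# Its output is scaled by surface volume, then by a base volume mixed per listener affiliation (enemy loudest, non-enemy quietest).
# It borrows CoveFern beam impact attenuation, caps concurrency, and keeps playing when silent.
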MetaSoundSource /WildEstateWeapons/Gameplay/CoveSpice/Audio/IP_MSS_CoveSpice_Impact_Loop
inputs [
 Base Volume {type=Float}
 Character.Affiliation {type=Enum:FortAudioAffiliation}
 Character.Health {type=Float}
 Character.UniqueID {type=Int32}
 Surfaces.CrossfadeTime {type=Time}
 Surfaces.FadeOut {type=Trigger}
 Surfaces.OnSurfaceChanged {type=Trigger}
 Surfaces.Pitch {type=Float}
 Surfaces.Sound {type=WaveAsset}
 Surfaces.Volume {type=Float}
 UE.Source.OnPlay {type=Trigger}
]
interfaces [Fort.Character:1.0, UE.OutputFormat.Mono:1.0, UE.Source:1.1]
outputs ["UE.OutputFormat.Mono.Audio:0" {type=Audio}]
pages [Default {
  input_defaults {
   Base Volume=3
   Character.Affiliation=0
   Character.Health=0
   Character.UniqueID=0
   Surfaces.CrossfadeTime=0.3
   Surfaces.FadeOut=null
   Surfaces.OnSurfaceChanged=null
   Surfaces.Pitch=0
   Surfaces.Sound=null
   Surfaces.Volume=1
   UE.Source.OnPlay=null
  }
  nodes [Multiply_0 = UE.Multiply:Audio by Float(
    AdditionalOperands=&MSP_AffiliationMixFloat(In Value=^Base Volume, Affiliation=^Character.Affiliation) {also {defaults {Enemy=1.25, Local=0.8, Non-Enemy=0.5}}}.Out
    PrimaryOperand=UE.Multiply:Audio by Float(
     AdditionalOperands=^Surfaces.Volume
     PrimaryOperand=/FNGameplayCues/Sounds/Fort_Audio_Blueprints/SurfaceTypeSound/MSP_SurfaceLoopSwapper(
      CrossfadeTime=^Surfaces.CrossfadeTime
      FadeOutAllSounds=^Surfaces.FadeOut
      PlayNextSound=^Surfaces.OnSurfaceChanged
      PitchShift=^Surfaces.Pitch
      NextSoundToPlay=^Surfaces.Sound
     ).Out
    ).Out
   )]
  links [{dst="@UE.OutputFormat.Mono.Audio:0", src=Multiply_0.Out}]
 }]
properties {
 assetImportData=&self:AssetImportData
 attenuationSettings=&/IP_ATT_CoveFern_Beam_Impact
 bLooping=true
 bOverrideConcurrency=true
 duration=10000
 maxDistance=15000
 sampleRate=48000
 soundClassObject=&Fort_Impacts
 virtualizationMode=PlayWhenSilent
 volume=1.3
 concurrencyOverrides {maxCount=3}
}
