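# Stereo MetaSound source for the looping first-person fire sound of the CoveSpice raygun.
# On play, a wave player starts the loop wave at a random time within its duration.
# An ADSR envelope fades both channels in through multiply nodes, and the player's On Finished drives the one-shot finished output.
# The Start Sounds array and random pitch and volume inputs are declared but unconnected.
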MetaSoundSource /WildEstateWeapons/Gameplay/CoveSpice/Audio/IP_MSS_CoveSpice_Raygun_Fire_Loop_1P
inputs [
 Loop Sound {type=WaveAsset}
 Random.Pitch.Base {type=Float}
 Random.Pitch.Spread {type=Float}
 Random.Volume.Base {type=Float}
 Random.Volume.Spread {type=Float}
 Start Sounds {type=WaveAsset:Array}
 UE.Source.OnPlay {type=Trigger}
]
interfaces [UE.OutputFormat.Stereo:1.0, UE.Source:1.1]
outputs [
 "UE.OutputFormat.Stereo.Audio:0" {type=Audio}
 "UE.OutputFormat.Stereo.Audio:1" {type=Audio}
 UE.Source.OneShot.OnFinished {type=Trigger}
]
pages [Default {
  input_defaults {
   Loop Sound=/WildEstateWeapons/Gameplay/CoveSpice/Audio/Waves/IP_CoveSpice_Raygun_Fire_Loop_01
   Random.Pitch.Base=0
   Random.Pitch.Spread=0
   Random.Volume.Base=0
   Random.Volume.Spread=0
   UE.Source.OnPlay=null
   Start Sounds [
    /WildEstateWeapons/Gameplay/CoveSpice/Audio/Waves/IP_CoveSpice_Raygun_Fire_ON_01
    /WildEstateWeapons/Gameplay/CoveSpice/Audio/Waves/IP_CoveSpice_Raygun_Fire_ON_02
    /WildEstateWeapons/Gameplay/CoveSpice/Audio/Waves/IP_CoveSpice_Raygun_Fire_ON_03
   ]
  }
  nodes [
   ADSREnvelopeADSREnvelope_0 = ADSR Envelope.ADSR Envelope:Float(Trigger Attack=^UE.Source.OnPlay) {also {defaults {Attack Time=0.4, Sustain Level=1}}}
   GetWaveDuration_0 = "UE.Get Wave Duration: "(Wave=^Loop Sound)
   WavePlayer_0 = UE.Wave Player:Stereo(Wave Asset=^Loop Sound, Play=^UE.Source.OnPlay, Start Time=UE.RandomTime(Next=^UE.Source.OnPlay, Max=@GetWaveDuration_0.Duration).Value) {also {defaults {Loop="<MetaSoundBuilderResult.SUCCEEDED: 0>"}, version [1,1]}}
   Multiply_0 = UE.Multiply:Audio by Float(AdditionalOperands=@ADSREnvelopeADSREnvelope_0.Out Envelope, PrimaryOperand=@WavePlayer_0.Out Right)
   Multiply_1 = UE.Multiply:Audio by Float(AdditionalOperands=@ADSREnvelopeADSREnvelope_0.Out Envelope, PrimaryOperand=@WavePlayer_0.Out Left)
  ]
  links [
   {dst="@UE.OutputFormat.Stereo.Audio:1", src=Multiply_0.Out}
   {dst="@UE.OutputFormat.Stereo.Audio:0", src=Multiply_1.Out}
   {dst="@UE.Source.OneShot.OnFinished", src=WavePlayer_0.On Finished}
  ]
 }]
properties {
 assetImportData=&self:AssetImportData
 attenuationSettings=&Fort_Weapon_Projectile_FirstPerson_Athena_Att
 bLooping=true
 duration=10000
 numChannels=2
 outputFormat=Stereo
 sampleRate=48000
 soundClassObject=&Fort_WeaponRanged_Fire_1P_PSM
 concurrencySet [&Weapon_Gunfire_Local_Conc]
}
