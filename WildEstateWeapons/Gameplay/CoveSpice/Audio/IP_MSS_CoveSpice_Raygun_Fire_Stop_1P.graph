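# One-shot mono MetaSound for the first-person stop sound when the CoveSpice Raygun ceases firing.
# It reuses the shared MSS_Weapon_Reload_Start graph, feeding it three fire-off waves, master volume and random pitch settings.
# That graph's audio passes through a short Delay node to the output, while its finished trigger bypasses the delay.
# The source uses local gunfire concurrency and first-person weapon-fire volume modulation.
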MetaSoundSource /WildEstateWeapons/Gameplay/CoveSpice/Audio/IP_MSS_CoveSpice_Raygun_Fire_Stop_1P
inputs [
 Character.Affiliation {type=Enum:FortAudioAffiliation}
 Character.Health {type=Float}
 Character.UniqueID {type=Int32}
 MasterVolume {type=Float}
 Random.PitchShift.Base {type=Float}
 Random.PitchShift.Spread {type=Float}
 Sounds {type=WaveAsset:Array}
 UE.Source.OnPlay {type=Trigger}
]
interfaces [Fort.Character:1.0, UE.OutputFormat.Mono:1.0, UE.Source.OneShot:1.0, UE.Source:1.1]
outputs [
 "UE.OutputFormat.Mono.Audio:0" {type=Audio}
 UE.Source.OneShot.OnFinished {type=Trigger}
]
pages [Default {
  input_defaults {
   Character.Affiliation=0
   Character.Health=0
   Character.UniqueID=0
   MasterVolume=1
   Random.PitchShift.Base=0
   Random.PitchShift.Spread=0.5
   UE.Source.OnPlay=null
   Sounds [
    /WildEstateWeapons/Gameplay/CoveSpice/Audio/Waves/IP_CoveSpice_Raygun_Fire_OFF_01
    /WildEstateWeapons/Gameplay/CoveSpice/Audio/Waves/IP_CoveSpice_Raygun_Fire_OFF_02
    /WildEstateWeapons/Gameplay/CoveSpice/Audio/Waves/IP_CoveSpice_Raygun_Fire_OFF_03
   ]
  }
  nodes [
   MSS_Weapon_Reload_Start_0 = /FortMetasound/Weapons/MSS_Weapon_Reload_Start(
    MasterVolume=^MasterVolume
    Random.PitchShift.Base=^Random.PitchShift.Base
    Random.PitchShift.Spread=^Random.PitchShift.Spread
    Sounds=^Sounds
    UE.Source.OnPlay=^UE.Source.OnPlay
   )
   Delay_0 = UE.Delay:Audio(In=@MSS_Weapon_Reload_Start_0."UE.OutputFormat.Mono.Audio:0") {also {defaults {Delay Time=0.1}}}
  ]
  links [
   {dst="@UE.OutputFormat.Mono.Audio:0", src=Delay_0.Out}
   {dst="@UE.Source.OneShot.OnFinished", src=MSS_Weapon_Reload_Start_0.UE.Source.OneShot.OnFinished}
  ]
 }]
properties {
 assetImportData=&self:AssetImportData
 sampleRate=48000
 soundClassObject=&Fort_WeaponRanged_Fire_1P_PSM
 volume=0.5
 concurrencySet [&Weapon_Gunfire_Local_Conc]
 modulationSettings {
  volumeRouting=Union
  volumeModulationDestination {bEnableModulation=true, modulators [/FortMetaSoundWeapon/Items/Modulation/PP_Weapon_Fire_1p_Vol]}
 }
}
