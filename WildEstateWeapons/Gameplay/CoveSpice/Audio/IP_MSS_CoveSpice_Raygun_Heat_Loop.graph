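# Stereo MetaSound source for the CoveSpice Raygun heat loop, using first-person weapon sound class and attenuation settings.
# On play it picks random volume and pitch, then a random wave from Sounds, and starts the wave player with pitch converted to semitones.
# An ADSR envelope and the random volume scale both channels.
# Release Time feeds the envelope but no release trigger is wired, and bUseRandomStartTime is unused.
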MetaSoundSource /WildEstateWeapons/Gameplay/CoveSpice/Audio/IP_MSS_CoveSpice_Raygun_Heat_Loop
inputs [
 Attack Time {type=Time}
 bUseRandomStartTime {type=Bool}
 Loop {type=Bool}
 Max Pitch {type=Float}
 Max Volume {type=Float}
 Min Pitch {type=Float}
 Min Volume {type=Float}
 Release Time {type=Time}
 Sounds {type=WaveAsset:Array}
 UE.Source.OnPlay {type=Trigger}
]
interfaces [UE.OutputFormat.Stereo:1.0, UE.Source:1.1]
outputs [
 "UE.OutputFormat.Stereo.Audio:0" {type=Audio}
 "UE.OutputFormat.Stereo.Audio:1" {type=Audio}
]
pages [Default {
  input_defaults {
   Attack Time=0.2
   Loop="<MetaSoundBuilderResult.SUCCEEDED: 0>"
   Max Pitch=1
   Max Volume=1
   Min Pitch=1
   Min Volume=0.9
   Release Time=0.2
   UE.Source.OnPlay=null
   bUseRandomStartTime=null
   Sounds [/WildEstateWeapons/Gameplay/CoveSpice/Audio/Waves/IP_CoveSpice_Raygun_Heat_Loop_01]
  }
  nodes [
   InitVariable.Time()
   n1 = UE.RandomFloat(Max=^Max Volume, Min=^Min Volume, Next=^UE.Source.OnPlay)
   n2 = UE.RandomFloat(Max=^Max Pitch, Min=^Min Pitch, Next=@n1.On Next)
   ArrayRandomGet_0 = Array.Random Get:WaveAsset:Array(In Array=^Sounds, Next=@n2.On Next) {also {defaults {No Repeats=-1, Weights []}}}
   FrequencyMultipliertoSemitone_0 = UE.Frequency Multiplier to Semitone:Float(Frequency Multiplier=UE.Reroute(Value=@n2.Value).Value)
   n3 = InitVariable.Float()
   n4 = InitVariable.WaveAsset()
   WavePlayer_0 = UE.Wave Player:Stereo(
    Loop=^Loop
    Play=@ArrayRandomGet_0.On Next
    Pitch Shift=VariableAccessor.Float(Variable=VariableMutator.Float(Value=@FrequencyMultipliertoSemitone_0.Semitones, Variable=@n3.Variable).Variable).Value
    Wave Asset=VariableAccessor.WaveAsset(Variable=VariableMutator.WaveAsset(Value=@ArrayRandomGet_0.Value, Variable=@n4.Variable).Variable).Value
   ) {also {version [1,1]}}
   ADSREnvelopeADSREnvelope_0 = ADSR Envelope.ADSR Envelope:Audio(Attack Time=^Attack Time, Release Time=^Release Time, Trigger Attack=^UE.Source.OnPlay) {also {defaults {Decay Time=1, Sustain Level=1}}}
   n5 = VariableAccessor.Audio(Variable=VariableMutator.Audio(Value=@ADSREnvelopeADSREnvelope_0.Out Envelope, Variable=InitVariable.Audio().Variable).Variable)
   n6 = VariableAccessor.Float(Variable=VariableMutator.Float(Variable=InitVariable.Float().Variable, Value=@n1.Value).Variable)
   Multiply_2 = UE.Multiply:Audio by Float(
    PrimaryOperand=UE.Multiply:Audio(AdditionalOperands=UE.Reroute(Value=@WavePlayer_0.Out Right).Value, PrimaryOperand=@n5.Value).Out
    AdditionalOperands=@n6.Value
   )
   Multiply_3 = UE.Multiply:Audio by Float(
    PrimaryOperand=UE.Multiply:Audio(PrimaryOperand=UE.Reroute(Value=@WavePlayer_0.Out Left).Value, AdditionalOperands=@n5.Value).Out
    AdditionalOperands=@n6.Value
   )
  ]
  links [{dst="@UE.OutputFormat.Stereo.Audio:1", src=Multiply_2.Out}, {dst="@UE.OutputFormat.Stereo.Audio:0", src=Multiply_3.Out}]
 }]
properties {
 assetImportData=&self:AssetImportData
 attenuationSettings=&Fort_Weapon_Projectile_FirstPerson_Athena_Att
 bLooping=true
 duration=10000
 numChannels=2
 outputFormat=Stereo
 sampleRate=48000
 soundClassObject=&Fort_WeaponRanged_Fire_1P_PSM
 concurrencySet [&CS_Fort_WeaponFoley]
}
