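# Stereo one-shot MetaSound for the third-person CoveSpice Raygun shock sound.
# On play, a shuffle node picks one of four wave variations and feeds a stereo wave player, which drives both output channels and signals finish when done.
# It uses a borrowed status-effect attenuation, the player elemental damage sound class, and a concurrency override that caps simultaneous instances and sets a retrigger time.
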MetaSoundSource /WildEstateWeapons/Gameplay/CoveSpice/Audio/IP_MSS_CoveSpice_Raygun_Shock_3P
inputs [
 Sounds {type=WaveAsset:Array}
 UE.Source.OnPlay {type=Trigger}
]
interfaces [UE.OutputFormat.Stereo:1.0, UE.Source.OneShot:1.0, UE.Source:1.1]
outputs [
 "UE.OutputFormat.Stereo.Audio:0" {type=Audio}
 "UE.OutputFormat.Stereo.Audio:1" {type=Audio}
 UE.Source.OneShot.OnFinished {type=Trigger}
]
pages [Default {
  input_defaults {
   UE.Source.OnPlay=null
   Sounds [
    /WildEstateWeapons/Gameplay/CoveSpice/Audio/Waves/IP_CoveSpice_Raygun_Shock_Oneshot_01
    /WildEstateWeapons/Gameplay/CoveSpice/Audio/Waves/IP_CoveSpice_Raygun_Shock_Oneshot_02
    /WildEstateWeapons/Gameplay/CoveSpice/Audio/Waves/IP_CoveSpice_Raygun_Shock_Oneshot_03
    /WildEstateWeapons/Gameplay/CoveSpice/Audio/Waves/IP_CoveSpice_Raygun_Shock_Oneshot_04
   ]
  }
  nodes [
   n1 = Array.Shuffle:WaveAsset:Array(In Array=^Sounds, Next=^UE.Source.OnPlay) {also {defaults {Enable Shared State="<MetaSoundBuilderResult.SUCCEEDED: 0>"}}}
   WavePlayer_0 = UE.Wave Player:Stereo(Play=@n1.On Next, Wave Asset=@n1.Value) {also {version [1,1]}}
  ]
  links [
   {dst="@UE.Source.OneShot.OnFinished", src=WavePlayer_0.On Finished}
   {dst="@UE.OutputFormat.Stereo.Audio:0", src=WavePlayer_0.Out Left}
   {dst="@UE.OutputFormat.Stereo.Audio:1", src=WavePlayer_0.Out Right}
  ]
 }]
properties {
 assetImportData=&self:AssetImportData
 attenuationSettings=/FNGameplayCues/Sounds/Attenuation/Fort_Ability_Status_Effect_Looping_Stereo_Att
 bOverrideConcurrency=true
 maxDistance=3500
 numChannels=2
 outputFormat=Stereo
 sampleRate=48000
 soundClassObject=/Game/Sounds/FortSoundClassesAndModes/Fort_ElementalDamage_Player_PSM
 volume=0.7
 concurrencyOverrides {maxCount=4, retriggerTime=0.2}
}
