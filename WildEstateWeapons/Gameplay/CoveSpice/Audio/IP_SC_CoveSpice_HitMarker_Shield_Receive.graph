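# Sound cue for the CoveSpice weapon's hit marker when a shield hit is received.
# Randomly picks one of three equally weighted shield hitmarker waves, then varies pitch through a Modulator at constant volume.
# Plays at reduced volume through the received shield hit notification sound class and shared receive attenuation.
# A concurrency override allows one instance, blocking new plays rather than cutting the current one, with a short retrigger delay.
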SoundCue /WildEstateWeapons/Gameplay/CoveSpice/Audio/IP_SC_CoveSpice_HitMarker_Shield_Receive
root=Modulator_0
nodes [
 {id=Modulator_0, children [Random_0], properties {pitchMax=1.1, volumeMax=1, volumeMin=1}}
 {id=Random_0, children [WavePlayer_4,WavePlayer_5,WavePlayer_6], properties {bShouldExcludeFromBranchCulling=true, weights [1,1,1]}}
 {id=WavePlayer_4, properties {soundWaveAssetPtr=/WildEstateWeapons/Gameplay/CoveSpice/Audio/Waves/IP_CoveSpice_Hitmarker_Shield_01}}
 {id=WavePlayer_5, properties {soundWaveAssetPtr=/WildEstateWeapons/Gameplay/CoveSpice/Audio/Waves/IP_CoveSpice_Hitmarker_Shield_02}}
 {id=WavePlayer_6, properties {soundWaveAssetPtr=/WildEstateWeapons/Gameplay/CoveSpice/Audio/Waves/IP_CoveSpice_Hitmarker_Shield_03}}
]
properties {
 attenuationSettings=/Game/Athena/Sounds/HitNotifications/Fort_HitMarker_Receive_Att
 bOverrideConcurrency=true
 maxDistance=2097152
 soundClassObject=/Game/Sounds/FortSoundClassesAndModes/Fort_2DHitNotifications_Receive_Shield
 volumeMultiplier=0.5
 concurrencyOverrides {maxCount=1, resolutionRule=PreventNew, retriggerTime=0.3}
 concurrencySet [&HitNotifications_TwoMaximum_StopOldest_VolumeScale]
}
