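# Third-person looping fire SoundCue for the CoveSpice ray gun, limited by local and global gunfire concurrency.
# It mixes close, medium and distant layers, each with its own attenuation and a fire sound class override, the distant one using the Distant class.
# Each layer pairs a random one-of-three "ON" onset with a loop wave given a few-millisecond envelope fade-in.
# Only the close loop is delayed, by a fixed interval, behind its onset.
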SoundCue /WildEstateWeapons/Gameplay/CoveSpice/Audio/IP_SC_CoveSpice_RayGun_FireLoop_3P
root=Mixer_17
nodes [
 {
  id=Attenuation_0
  children [Mixer_1]
  properties {attenuationSettings=&Fort_Weapon_Projectile_Medium_Mono_Athena_Att}
 }
 {id=Attenuation_1, children [Mixer_0], properties {attenuationSettings=&Fort_Weapon_Projectile_Close_Athena_Att}}
 {
  id=Attenuation_2
  children [Mixer_2]
  properties {attenuationSettings=&Fort_Weapon_Projectile_Distant_Mono_Athena_Att}
 }
 {id=Delay_0, children [Enveloper_0], properties {delayMax=0.3, delayMin=0.3}}
 {
  id=Enveloper_0
  children [WavePlayer_27]
  properties {pitchCurve {editorCurveData {keys [{time=0}]}}, volumeCurve {editorCurveData {keys [<at -0.0001 -0.0176>,<at 0.0038 1>]}}}
 }
 {
  id=Enveloper_1
  children [WavePlayer_10]
  properties {pitchCurve {editorCurveData {keys [{time=0}]}}, volumeCurve {editorCurveData {keys [<at -0.0001 -0.0176>,<at 0.0038 1>]}}}
 }
 {
  id=Enveloper_2
  children [WavePlayer_14]
  properties {pitchCurve {editorCurveData {keys [{time=0}]}}, volumeCurve {editorCurveData {keys [<at -0.0001 -0.0176>,<at 0.0038 1>]}}}
 }
 {id=Mixer_0, children [Random_2,Delay_0], properties {inputVolume [1,1]}}
 {id=Mixer_1, children [Random_6,Enveloper_1], properties {inputVolume [0.6,0.5]}}
 {id=Mixer_17, children [SoundClass_1,SoundClass_0,SoundClass_2], properties {inputVolume [1,0.7,0.4]}}
 {id=Mixer_2, children [Random_4,Enveloper_2], properties {inputVolume [0.6,0.5]}}
 {id=Modulator_0, children [Attenuation_2], properties {pitchMax=1, pitchMin=1}}
 {id=Random_2, children [WavePlayer_16, WavePlayer_25, WavePlayer_26], properties {weights [1,1,1]}}
 {id=Random_4, children [WavePlayer_11, WavePlayer_12, WavePlayer_13], properties {weights [1,1,1]}}
 {id=Random_6, children [WavePlayer_0,WavePlayer_1,WavePlayer_9], properties {weights [1,1,1]}}
 {
  id=SoundClass_0
  children [Attenuation_0]
  properties {soundClassOverride=&Fort_WeaponRanged_Fire_3P_PSM}
 }
 {
  id=SoundClass_1
  children [Attenuation_1]
  properties {soundClassOverride=&Fort_WeaponRanged_Fire_3P_PSM}
 }
 {
  id=SoundClass_2
  children [Modulator_0]
  properties {soundClassOverride=&Fort_WeaponRanged_Fire_Distant}
 }
 {id=WavePlayer_0, properties {soundWaveAssetPtr=/WildEstateWeapons/Gameplay/CoveSpice/Audio/Waves/IP_CoveSpice_Raygun_Fire_Medium_ON_01}}
 {id=WavePlayer_1, properties {soundWaveAssetPtr=/WildEstateWeapons/Gameplay/CoveSpice/Audio/Waves/IP_CoveSpice_Raygun_Fire_Medium_ON_02}}
 {id=WavePlayer_10, properties {soundWaveAssetPtr=/WildEstateWeapons/Gameplay/CoveSpice/Audio/Waves/IP_CoveSpice_Raygun_Fire_Medium_Loop_01}}
 {id=WavePlayer_11, properties {soundWaveAssetPtr=/WildEstateWeapons/Gameplay/CoveSpice/Audio/Waves/IP_CoveSpice_Raygun_Fire_Far_ON_01}}
 {id=WavePlayer_12, properties {soundWaveAssetPtr=/WildEstateWeapons/Gameplay/CoveSpice/Audio/Waves/IP_CoveSpice_Raygun_Fire_Far_ON_02}}
 {id=WavePlayer_13, properties {soundWaveAssetPtr=/WildEstateWeapons/Gameplay/CoveSpice/Audio/Waves/IP_CoveSpice_Raygun_Fire_Far_ON_03}}
 {id=WavePlayer_14, properties {soundWaveAssetPtr=/WildEstateWeapons/Gameplay/CoveSpice/Audio/Waves/IP_CoveSpice_Raygun_Fire_Far_Loop_01}}
 {id=WavePlayer_16, properties {soundWaveAssetPtr=/WildEstateWeapons/Gameplay/CoveSpice/Audio/Waves/IP_CoveSpice_Raygun_Fire_ON_01}}
 {id=WavePlayer_25, properties {soundWaveAssetPtr=/WildEstateWeapons/Gameplay/CoveSpice/Audio/Waves/IP_CoveSpice_Raygun_Fire_ON_02}}
 {id=WavePlayer_26, properties {soundWaveAssetPtr=/WildEstateWeapons/Gameplay/CoveSpice/Audio/Waves/IP_CoveSpice_Raygun_Fire_ON_03}}
 {id=WavePlayer_27, properties {soundWaveAssetPtr=/WildEstateWeapons/Gameplay/CoveSpice/Audio/Waves/IP_CoveSpice_Raygun_Fire_Loop_01}}
 {id=WavePlayer_9, properties {soundWaveAssetPtr=/WildEstateWeapons/Gameplay/CoveSpice/Audio/Waves/IP_CoveSpice_Raygun_Fire_Medium_ON_03}}
]
properties {
 bDebug=true
 maxDistance=27500
 volumeMultiplier=1
 concurrencySet [&Weapon_Gunfire_Local_Conc, &Weapon_Gunfire_Global_Conc]
}
