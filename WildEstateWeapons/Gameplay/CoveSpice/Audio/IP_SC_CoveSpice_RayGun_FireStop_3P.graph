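# Sound cue for the third-person fire-stop sound of the CoveSpice ray gun.
# A mixer blends a close layer with a slightly quieter medium-distance layer, each picking randomly from three equally weighted "OFF" waves through its own distance attenuation.
# Both layers override to the third-person ranged weapon fire sound class.
# Local and global gunfire concurrency sets limit playback, and the cue's debug flag is on.
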SoundCue /WildEstateWeapons/Gameplay/CoveSpice/Audio/IP_SC_CoveSpice_RayGun_FireStop_3P
root=Mixer_17
nodes [
 {
  id=Attenuation_0
  children [Random_10]
  properties {attenuationSettings=&Fort_Weapon_Projectile_Medium_Mono_Athena_Att}
 }
 {id=Attenuation_1, children [Random_9], properties {attenuationSettings=&Fort_Weapon_Projectile_Close_Athena_Att}}
 {id=Mixer_17, children [SoundClass_1,SoundClass_0], properties {inputVolume [1,0.7]}}
 {id=Random_10, children [WavePlayer_0,WavePlayer_1,WavePlayer_2], properties {weights [1,1,1]}}
 {id=Random_9, children [WavePlayer_12, WavePlayer_13, WavePlayer_14], properties {weights [1,1,1]}}
 {
  id=SoundClass_0
  children [Attenuation_0]
  properties {soundClassOverride=&Fort_WeaponRanged_Fire_3P_PSM}
 }
 {
  id=SoundClass_1
  children [Attenuation_1]
  properties {soundClassOverride=&Fort_WeaponRanged_Fire_3P_PSM}
 }
 {id=WavePlayer_0, properties {soundWaveAssetPtr=/WildEstateWeapons/Gameplay/CoveSpice/Audio/Waves/IP_CoveSpice_Raygun_Fire_Medium_OFF_03}}
 {id=WavePlayer_1, properties {soundWaveAssetPtr=/WildEstateWeapons/Gameplay/CoveSpice/Audio/Waves/IP_CoveSpice_Raygun_Fire_Medium_OFF_02}}
 {id=WavePlayer_12, properties {soundWaveAssetPtr=/WildEstateWeapons/Gameplay/CoveSpice/Audio/Waves/IP_CoveSpice_Raygun_Fire_OFF_01}}
 {id=WavePlayer_13, properties {soundWaveAssetPtr=/WildEstateWeapons/Gameplay/CoveSpice/Audio/Waves/IP_CoveSpice_Raygun_Fire_OFF_02}}
 {id=WavePlayer_14, properties {soundWaveAssetPtr=/WildEstateWeapons/Gameplay/CoveSpice/Audio/Waves/IP_CoveSpice_Raygun_Fire_OFF_03}}
 {id=WavePlayer_2, properties {soundWaveAssetPtr=/WildEstateWeapons/Gameplay/CoveSpice/Audio/Waves/IP_CoveSpice_Raygun_Fire_Medium_OFF_01}}
]
properties {
 bDebug=true
 maxDistance=10579.2871
 volumeMultiplier=0.5
 concurrencySet [&Weapon_Gunfire_Local_Conc, &Weapon_Gunfire_Global_Conc]
}
