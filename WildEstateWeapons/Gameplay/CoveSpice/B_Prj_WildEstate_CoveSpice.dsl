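; Projectile for the Wild Estate Cove Spice weapon that applies gameplay effects to the actor it hits when it stops.
; BeginPlay reads a scalable lifetime and uses it for both the actor lifespan and the max lifespan on stop.
; OnStop stores the hit, then ApplyPlayerEffect applies each configured effect to the hit actor's ability system, with a context carrying instigator, current weapon and hit location as origin.
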(event EventBeginPlay
  (bind _cove_spice_projectile_lifetime (Variables|Default|GetCove%20Spice%20Projectile%20Lifetime))
  (bind _returnvalue (ScalableFloat|GetValueatLevel _cove_spice_projectile_lifetime))
  (|CallParentFunction)
  (Variables|Projectile|SetMaxLifespanOnStop _returnvalue)
  (Actor|SetLifeSpan _returnvalue))

(event EventOnStop
  (params (Hit :type Hit%20Result%20Structure%20%28by%20ref%29))
  (bind (_bblockinghit _binitialoverlap _time _distance _location _impactpoint _normal _impactnormal _physmat _hitactor) (Collision|BreakHitResult (Variables|Default|GetHit)))
  (|CallParentFunction Hit)
  (Variables|Default|SetHit Hit)
  (CallFunction|ApplyPlayerEffect _hitactor))

(fn ApplyPlayerEffect
  (params (Apply%20to%20Actor :type Actor%20Object%20Reference))
  (bind (_bblockinghit _binitialoverlap _time _distance _location) (Collision|BreakHitResult (Variables|Default|GetHit)))
  (bind _instigator (Variables|Actor|GetInstigator))
  (bind _asplayer_pawn_athena (Utilities|Casting|CastToPlayerPawn_Athena _instigator))
  (bind _output_get (LocalVariables|SetEffect%20Context (Class|AthenaFunctionLibrary|MakeSimpleFortEffectContext _instigator (Variables|/Script/FortniteGame.FortPawn|GetCurrentWeapon _asplayer_pawn_athena))))
  (Ability|EffectContext|SetOrigin _output_get _location)
  (for _array_element (Variables|Default|GetEffects%20to%20apply)
    (GameplayEffects|ApplyGameplayEffectToSelf (Ability|GetAbilitySystemComponent (Variables|GetApply%20to%20Actor)) _array_element 0.0 (LocalVariables|GetEffect%20Context))))

(fn ConstructionScript)
