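; Weapon Blueprint for the Wild Estate CoveSpice raygun, a continuous beam that overheats and ramps damage while firing.
; A timer has the server derive replicated charge percent from stacking damage GameplayEffect count, feeding beam, muzzle and impact FX.
; Local rumble and camera shake scale with charge, while the replicated trace hit drives impact FX and child-actor beam audio.
; It also handles montages, first-equip audio and a looping siphon GameplayCue for the exotic variant.
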(event EventBeginPlay
  (|CallParentFunction)
  (Networking|CanExecuteCosmeticEvents
    (:False
      (Components|DestroyComponent (Variables|Default|GetBeamAudioComponent)))))

(event FortWeapon|EventOnEquippedAndAttached
  (bind _beam (Variables|Default|GetBeam))
  ;[0 0.07229 0.15 0.5] Setup on Equip
  (|CallParentFunction)
  (Effects|InitializeBeamFX _beam)
  (Rendering|SetRenderCustomDepth _beam true)
  (Components|Activation|Deactivate _beam)
  (bind _returnvalue (Utilities|Time|SetTimerbyEvent (EventDispatchers|CreateEvent :!event Custom|Update%20Beam%20Charge) 0.1 true true))
  (Variables|Default|SetBeamUpdateTimer _returnvalue)
  (if (not (GameplayTags|HasMatchingGameplayTag (Utilities|Casting|ToGameplayTagAssetInterface (Game|GetInstigator)) "(TagName=\"Gameplay.Status.Wild.Estate.CoveSpice.FirstEquip\")"))
    (Variables|Default|SetEquipSound "/WildEstateWeapons/Gameplay/CoveSpice/Audio/IP_MSS_CoveSpice_Raygun_Equip_Tentacle.IP_MSS_CoveSpice_Raygun_Equip_Tentacle")
    (else
      (Variables|Default|SetEquipSound "/WildEstateWeapons/Gameplay/CoveSpice/Audio/IP_MSS_CoveSpice_Raygun_Equip_Short.IP_MSS_CoveSpice_Raygun_Equip_Short")))
  (Audio|SpawnSoundAttached (Variables|Default|GetEquipSound) (Variables|Visual|GetWeaponMesh)))

(event FortWeapon|EventOnUnEquippedAndDetached
  ;[0.15 0 0.003439 0.5] Unequip
  (|CallParentFunction)
  (Utilities|Time|ClearandInvalidateTimerbyHandle (Variables|Default|GetBeamUpdateTimer)))

(event EventOnStartCharge
  (bind _weapon_chargeup_montage (Variables|Animation|GetWeapon%20Chargeup%20Montage))
  (bind _player_chargeup_montage (Variables|Animation|GetPlayer%20Chargeup%20Montage))
  ;[0.15 0 0.1344 0.5] On Start Charge
  (|CallParentFunction)
  (Utilities|IsValid _weapon_chargeup_montage
    (:Is%20Valid
      (Animation|Montage|MontagePlaywithBlendIn (Components|SkeletalMesh|GetAnimInstance (Variables|Visual|GetWeaponMesh)) _weapon_chargeup_montage (struct :CustomCurve 0 :BlendTime 0.2 :BlendOption "HermiteCubic") 1.0 "MontageLength" 0.0 false)
      (Utilities|IsValid _player_chargeup_montage
        (:Is%20Valid
          (bind _asfort_player_pawn (Utilities|Casting|CastToFortPlayerPawn (Game|GetInstigator)))
          (Animation|Montage|MontagePlaywithBlendIn (Components|SkeletalMesh|GetAnimInstance (Variables|/Script/Engine.Character|GetMesh _asfort_player_pawn)) _player_chargeup_montage (struct :CustomCurve 0 :BlendTime 0.2 :BlendOption "HermiteCubic") 1.0 "MontageLength" 0.0 false))))))

(event Overheating|EventOnStartOverheated
  ;[0.6302 0.2396 0 0.5] On Overheat
  (|CallParentFunction)
  (Niagara|SpawnSystemAttached "/WildEstateWeapons/Assets/FX/CoveSpice/NS_WildEstate_CoveSpice_Overheat.NS_WildEstate_CoveSpice_Overheat" (Variables|Visual|GetWeaponMesh) "root"))

(event Custom|Update%20Beam%20Charge
  (bind _returnvalue (CallFunction|GetBeamNiagaraComponent))
  (bind _charge_percent (Variables|DamageChargeUp|GetCharge%20Percent))
  ;[0.00011 0.15 0 0.5] Update Beam Charge
  (Utilities|IsValid _returnvalue
    (:Is%20Valid
      (CallFunction|UpdateChargePercent)
      (Networking|CanExecuteCosmeticEvents
        (:True
          (Niagara|SetNiagaraVariable(Float) _returnvalue "ChargePercent" _charge_percent)
          (Effects|Components|ParticleSystem|SetFloatParameter (Variables|Effects|Muzzle|GetSpawned%20Muzzle%20System) "ChargePercent" _charge_percent))))))

(event FortWeapon|EventOnPlayImpactFX
  (params (HitResult :type Hit%20Result%20Structure%20%28by%20ref%29) (ImpactPhysicalSurface :type EPhysicalSurface%20Enum :default "SurfaceType_Default") (SpawnedPSC :type FXSystem%20Component%20Object%20Reference))
  (bind (_bblockinghit _binitialoverlap _time _distance _location _impactpoint _normal) (Collision|BreakHitResult HitResult))
  (bind _charge_percent (Variables|DamageChargeUp|GetCharge%20Percent))
  (|CallParentFunction HitResult ImpactPhysicalSurface SpawnedPSC)
  (CallFunction|UpdateChargePercent)
  (Networking|CanExecuteCosmeticEvents
    (:True
      (bind _impactspawn (Variables|Default|GetImpactSpawn)
        (:then
          (Transformation|SetWorldTransform _impactspawn (Math|Transform|MakeTransform _location (Math|Conversions|RotationFromXVector _normal)) false false)
          (Effects|Components|ParticleSystem|SetFloatParameter _impactspawn "ChargePercent" _charge_percent)
          (Effects|Components|ParticleSystem|SetIntParameter _impactspawn "PhysicalSurface" (Math|Conversions|ToInteger(Byte) ImpactPhysicalSurface))
          (Effects|Components|ParticleSystem|SetFloatParameter SpawnedPSC "ChargePercent" _charge_percent))
        (:else)))))

(event EventOnPlayWeaponFireFX
  (params (bPersistentFire :type Boolean) (bSecondaryFire :type Boolean))
  (|CallParentFunction bPersistentFire bSecondaryFire)
  (CallFunction|StartFiringMontage)
  (CallFunction|StartDamageChargeRumble)
  (CallFunction|StartLocalCameraShake)
  (Variables|Default|SetClient_CurrentlyFiring true))

(event EventOnStopWeaponFireFX
  (|CallParentFunction)
  (bind _impactspawn (Variables|Default|GetImpactSpawn)
    (:then
      (Components|Activation|Deactivate _impactspawn))
    (:else
      _))
  (CallFunction|StopFiringMontage)
  (CallFunction|StopDamageChargeRumble)
  (CallFunction|StopCameraShake)
  (Class|BPICoveSpiceSiphon|StoppedHittingPawn)
  (Variables|Default|SetClient_CurrentlyFiring false))

(event EventPlayWeaponFireCustomAudio
  (params (bPersistentFire :type Boolean) (bSecondaryFire :type Boolean))
  ;[1 1 1] Beam Audio Implementation
  (|CallParentFunction bPersistentFire bSecondaryFire)
  (BeamAudio|PlayBeamAudio(Message) (Variables|/Script/Engine.ChildActorComponent|GetChildActor (Variables|Default|GetBeamAudioComponent)) true))

(event EventStopWeaponFireCustomAudio
  ;[1 1 1] Beam Audio Implementation
  (|CallParentFunction)
  (BeamAudio|PlayBeamAudio(Message) (Variables|/Script/Engine.ChildActorComponent|GetChildActor (Variables|Default|GetBeamAudioComponent)) false)
  (CallFunction|UpdateBeamAudio))

(event Custom|Start%20Firing%20Montage
  (bind _weapon_firing_montage (Variables|Animation|GetWeapon%20Firing%20Montage))
  (bind _player_firing_montage (Variables|Animation|GetPlayer%20Firing%20Montage))
  (Utilities|FlowControl|Sequence
    (:then_0
      ;[0.1128 0.15 0 0.5] weapon firing montage
      (Utilities|IsValid _weapon_firing_montage
        (:Is%20Valid
          (Animation|Montage|PlayMontage (Variables|Visual|GetWeaponMesh) _weapon_firing_montage))))
    (:then_1
      ;[0 0.15 0.03549 0.5] player firing montage
      (Utilities|IsValid _player_firing_montage
        (:Is%20Valid
          (bind _asfort_player_pawn (Utilities|Casting|CastToFortPlayerPawn (Game|GetInstigator)))
          (Animation|Montage|PlayMontage (Variables|/Script/Engine.Character|GetMesh _asfort_player_pawn) _player_firing_montage))))))

(event Custom|Stop%20Firing%20Montage
  (bind _weapon_finish_firing_montage (Variables|Animation|GetWeapon%20Finish%20Firing%20Montage))
  (bind _player_finish_firing_montage (Variables|Animation|GetPlayer%20Finish%20Firing%20Montage))
  ;[0.1373 0 0.15 0.5] Stop firing montages
  (Utilities|FlowControl|Sequence
    (:then_0
      (Utilities|IsValid _weapon_finish_firing_montage
        (:Is%20Valid
          (Animation|Montage|PlayMontage (Variables|Visual|GetWeaponMesh) _weapon_finish_firing_montage))))
    (:then_1
      (Utilities|IsValid _player_finish_firing_montage
        (:Is%20Valid
          (bind _asfort_player_pawn (Utilities|Casting|CastToFortPlayerPawn (Game|GetInstigator)))
          (Animation|Montage|PlayMontage (Variables|/Script/Engine.Character|GetMesh _asfort_player_pawn) _player_finish_firing_montage))))))

(event EventStartedHittingPawn
  (bind _exotic_healing_gcnl (Variables|Default|GetExotic%20Healing%20GCNL))
  ;[0.01169 0.15 0.02141 0.5] Start/Stop Siphon FX on player if exotic version of weapon
  (Networking|SwitchHasAuthority
    (:Authority
      (bind _asfort_player_pawn (Utilities|Casting|CastToFortPlayerPawn (Game|GetInstigator)))
      (bind _returnvalue (Ability|GameplayCue|IsGameplayCueActive _asfort_player_pawn _exotic_healing_gcnl))
      (if (not _returnvalue)
        (GameplayCue|AddGameplayCueOnActor(Looping) _asfort_player_pawn _exotic_healing_gcnl (Ability|GameplayCue|MakeGameplayCueParameters :NormalizedMagnitude 0.0 :RawMagnitude 0.0 :Location "0,0,0" :Normal "0,0,0" :Instigator _asfort_player_pawn :GameplayEffectLevel 1 :AbilityLevel 1 :bReplicateLocationWhenUsingMinimalRepProxy false))))))

(event EventStoppedHittingPawn
  ;[0.01169 0.15 0.02141 0.5] Start/Stop Siphon FX on player if exotic version of weapon
  (CallFunction|OnStoppedHittingPawn) ; need to call on server for the GCNL
)

(event Custom|OnStoppedHittingPawn
  (bind _exotic_healing_gcnl (Variables|Default|GetExotic%20Healing%20GCNL))
  ;[0.01169 0.15 0.02141 0.5] Start/Stop Siphon FX on player if exotic version of weapon
  (bind _asfort_player_pawn (Utilities|Casting|CastToFortPlayerPawn (Game|GetInstigator)))
  (bind _returnvalue (Ability|GameplayCue|IsGameplayCueActive _asfort_player_pawn _exotic_healing_gcnl))
  (if _returnvalue
    (GameplayCue|RemoveGameplayCueOnActor(Looping) _asfort_player_pawn _exotic_healing_gcnl (Ability|GameplayCue|MakeGameplayCueParameters :NormalizedMagnitude 0.0 :RawMagnitude 0.0 :Location "0,0,0" :Normal "0,0,0" :Instigator _asfort_player_pawn :GameplayEffectLevel 1 :AbilityLevel 1 :bReplicateLocationWhenUsingMinimalRepProxy false))))

(event Custom|Start%20Damage%20Charge%20Rumble
  (bind _returnvalue (CallFunction|GetPlayerController))
  ;[0.3516 0.1685 0.599 0.5] Controller Rumble Events (Start/Stop/Update)
  (Networking|CanExecuteCosmeticEvents
    (:True
      ;;[0.15 0.02035 0 0.5] Get owning local player controller
      (Utilities|IsValid _returnvalue
        (:Is%20Valid
          (if (Pawn|IsLocalPlayerController _returnvalue)
            (Variables|Default|SetOwning%20Controller _returnvalue)
            ;;[0 0 0 0.5] Start the rumble timer
            (bind _returnvalue_2 (Utilities|Time|SetTimerbyEvent (EventDispatchers|CreateEvent :!event Custom|Update%20Charge%20Damage%20Rumble) (ScalableFloat|GetValueatLevel (Variables|DamageChargeUp|GetDamage%20Charge%20Rumble%20Rate)) true))
            (Variables|DamageChargeUp|SetDamage%20Charge%20Rumble%20Timer _returnvalue_2)
            (Variables|Default|GetOwning%20Controller
              (:then
                (Game|Feedback|PlayDynamicForceFeedback (Variables|Default|GetOwning%20Controller) (Math|Float|Clamp(Float) (+ (* (Variables|DamageChargeUp|GetCharge%20Percent) (ScalableFloat|GetValueatLevel (Variables|DamageChargeUp|GetDamage%20Charge%20Rumble%20Multiplier))) (Variables|DamageChargeUp|GetBase%20Firing%20Rumble%20Value))) -1 true false true false))
              (:else))))))))

(event Custom|Update%20Charge%20Damage%20Rumble
  ;[0.3516 0.1685 0.599 0.5] Controller Rumble Events (Start/Stop/Update)
  (Networking|CanExecuteCosmeticEvents
    (:True
      (Variables|Default|GetOwning%20Controller
        (:then
          (Game|Feedback|PlayDynamicForceFeedback (Variables|Default|GetOwning%20Controller) (Math|Float|Clamp(Float) (+ (* (Variables|DamageChargeUp|GetCharge%20Percent) (ScalableFloat|GetValueatLevel (Variables|DamageChargeUp|GetDamage%20Charge%20Rumble%20Multiplier))) (Variables|DamageChargeUp|GetBase%20Firing%20Rumble%20Value))) -1 true false true false))
        (:else)))))

(event Custom|Stop%20Damage%20Charge%20Rumble
  ;[0.3516 0.1685 0.599 0.5] Controller Rumble Events (Start/Stop/Update)
  (Utilities|Time|ClearandInvalidateTimerbyHandle (Variables|DamageChargeUp|GetDamage%20Charge%20Rumble%20Timer))
  (Variables|Default|GetOwning%20Controller
    (:then
      (Game|Feedback|PlayDynamicForceFeedback (Variables|Default|GetOwning%20Controller) (Math|Float|Clamp(Float) (+ (* (Variables|DamageChargeUp|GetCharge%20Percent) (ScalableFloat|GetValueatLevel (Variables|DamageChargeUp|GetDamage%20Charge%20Rumble%20Multiplier))) (Variables|DamageChargeUp|GetBase%20Firing%20Rumble%20Value))) -1 true false true false))
    (:else)))

(event Custom|Start%20Local%20Camera%20Shake
  (bind _returnvalue (CallFunction|GetPlayerController))
  ;[0 0.12 0.15 0.5] Camera Shake Events (Start/Stop/Update)
  (Utilities|IsValid _returnvalue
    (:Is%20Valid
      (if (Pawn|IsLocalPlayerController _returnvalue)
        (bind _returnvalue_2 (CameraShakes|StartCameraShake :self (Variables|/Script/Engine.PlayerController|GetPlayerCameraManager _returnvalue) :ShakeClass (Variables|DamageChargeUp|GetCamera%20Shake) :Scale 1.0 :PlaySpace "CameraLocal"))
        (Variables|DamageChargeUp|SetCharge%20Up%20Camera%20Shake%20Inst _returnvalue_2)
        (bind _returnvalue_3 (Utilities|Time|SetTimerbyEvent (EventDispatchers|CreateEvent :!event Custom|On%20Update%20Camera%20Shake) (ScalableFloat|GetValueatLevel (Variables|DamageChargeUp|GetCharge%20Up%20Camera%20Shake%20Update%20Rate)) true))
        (Variables|DamageChargeUp|SetCamera%20Shake%20Timer _returnvalue_3)))))

(event Custom|On%20Update%20Camera%20Shake
  ;[0 0.12 0.15 0.5] Camera Shake Events (Start/Stop/Update)
  (Networking|CanExecuteCosmeticEvents
    (:True
      (bind _charge_up_camera_shake_inst (Variables|DamageChargeUp|GetCharge%20Up%20Camera%20Shake%20Inst)
        (:then
          (Variables|/Script/Engine.CameraShakeBase|SetShakeScale (Math|Float|Clamp(Float) (+ (* (Variables|DamageChargeUp|GetCharge%20Percent) (ScalableFloat|GetValueatLevel (Variables|DamageChargeUp|GetDamage%20Charge%20Camera%20Shake%20Multiplier))) (Variables|/Script/Engine.CameraShakeBase|GetShakeScale _charge_up_camera_shake_inst)) 0.0 2.0) _charge_up_camera_shake_inst))
        (:else)))))

(event Custom|Stop%20Camera%20Shake
  (bind _returnvalue (Game|GetPlayerCameraManager 0))
  (bind _charge_up_camera_shake_inst (Variables|DamageChargeUp|GetCharge%20Up%20Camera%20Shake%20Inst))
  ;[0 0.12 0.15 0.5] Camera Shake Events (Start/Stop/Update)
  (Utilities|Time|ClearandInvalidateTimerbyHandle (Variables|DamageChargeUp|GetCamera%20Shake%20Timer))
  (if (and (Utilities|IsValid _returnvalue :!class /Script/Engine.KismetSystemLibrary) (Utilities|IsValid _charge_up_camera_shake_inst :!class /Script/Engine.KismetSystemLibrary))
    (CameraShakes|StopCameraShake _returnvalue _charge_up_camera_shake_inst false)))

(fn GetCurrentDamageIncreasePercent
  (bind _returnvalue (Game|GetInstigator))
  (Utilities|IsValid _returnvalue
    (:Is%20Valid
      (return (Math|Float|SafeDivide (Math|Conversions|ToFloat(Integer) (GameplayEffects|GetGameplayEffectCount (Ability|GetAbilitySystemComponent _returnvalue) (Variables|DamageChargeUp|GetStacking%20Damage%20GE) 0 true)) (Math|Conversions|ToFloat(Integer) (Class|GetClassDefaults (Variables|DamageChargeUp|GetStacking%20Damage%20GE))))))))

(fn HandleImpactFX
  (bind (_bblockinghit _binitialoverlap _time _distance _location) (Collision|BreakHitResult (Variables|Default|GetBeam%20Trace%20Hit%20Result)))
  (bind (_bblockinghit_2 _binitialoverlap_2 _time_2 _distance_2 _location_2 _impactpoint _normal) (Collision|BreakHitResult (Variables|Default|GetBeam%20Trace%20Hit%20Result)))
  (bind _impactspawn (Variables|Default|GetImpactSpawn))
  (if (and (Collision|BreakHitResult (Variables|Default|GetBeam%20Trace%20Hit%20Result)) (Variables|Default|GetClient_CurrentlyFiring))
    (Variables|Default|GetImpactSpawn
      (:then
        ;[0.15 0 0.1369 0.5] Update Impact
        (if (not (Components|Activation|IsActive _impactspawn))
          (Components|Activation|Activate _impactspawn true))
        (Transformation|SetWorldLocation _impactspawn _location false true))
      (:else
        ;[0.1274 0.15 0.001613 0.5] If no current spawn FX, create one
        (if (FortWeapon|IsFiring)
          (bind _returnvalue (Effects|Components|ParticleSystem|SpawnFXAtLocation "/WildEstateWeapons/Assets/FX/CoveSpice/NS_WildEstate_CoveSpice_RibbonDecal.NS_WildEstate_CoveSpice_RibbonDecal" _location_2 (Math|Conversions|RotationFromXVector _normal) "1.000000,1.000000,1.000000" false))
          (bind _output_get (Variables|Default|SetImpactSpawn _returnvalue))
          (Rendering|SetRenderCustomDepth _output_get true))))
    (else
      ;[0.15 0.05863 0 0.5] If shooting the sky (or nothing) deactivate the impact
      (bind _impactspawn_2 (Variables|Default|GetImpactSpawn)
        (:then
          (Components|Activation|Deactivate _impactspawn_2))
        (:else)))))

(fn OnRep_BeamTraceHitResult
  (Networking|CanExecuteCosmeticEvents
    (:True
      (CallFunction|UpdateBeamAudio)
      (CallFunction|HandleImpactFX))))

(fn OnRep_ChargePercent)

(fn SelectRandomOverheatAnim
  (bind _overheat_animations (Variables|Animation|GetOverheat%20Animations))
  (bind _returnvalue (Utilities|Map|GetLastIndex _overheat_animations))
  (bind _returnvalue_2 (Math|Random|RandomIntegerinRange 0 _returnvalue))
  (bind (_key _value) (Utilities|Map|GetKeyValueByIndex _overheat_animations _returnvalue_2))
  (Variables|Animation|SetPlayer%20Overheat%20Montage _key)
  (Variables|Animation|SetWeapon%20Overheat%20Montage _value))

(fn UpdateBeamAudio
  (bind (_bblockinghit _binitialoverlap _time _distance _location _impactpoint _normal _impactnormal _physmat _hitactor) (Collision|BreakHitResult (Variables|Default|GetBeam%20Trace%20Hit%20Result)))
  (Networking|CanExecuteCosmeticEvents
    (:True
      (BeamAudio|UpdateBeamAudio(Message) (Variables|/Script/Engine.ChildActorComponent|GetChildActor (Variables|Default|GetBeamAudioComponent)) (Effects|GetMuzzleLocation -1) (Math|Vector|SelectVector _impactpoint _location _bblockinghit) _hitactor false _bblockinghit) ; Temp
    )))

(fn UpdateBeamAudioHitResult
  (params (Beam%20Hit%20Result :type Hit%20Result%20Structure))
  (Variables|Default|SetBeam%20Trace%20Hit%20Result Beam%20Hit%20Result))

(fn UpdateChargePercent
  (if (Networking|HasAuthority)
    (Variables|DamageChargeUp|SetCharge%20Percent (CallFunction|GetCurrentDamageIncreasePercent))
    (return)))

(fn ConstructionScript
  (|CallParentFunction))
