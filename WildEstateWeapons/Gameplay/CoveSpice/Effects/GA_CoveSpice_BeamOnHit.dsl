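; Gameplay ability for the WildEstate CoveSpice weapon that passes beam hit results to the weapon to update its beam audio.
; Activates from a gameplay event, runs the parent logic, then stores the event instigator as a player pawn.
; If the target data has a hit result, it casts that pawn's current weapon to B_WildEstate_CoveSpice and calls UpdateBeamAudioHitResult.
; A sequence then ends the ability immediately, making each event a one-shot update.
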(event Ability|EventActivateAbilityFromEvent
  (params (EventData :type Gameplay%20Event%20Data%20Structure%20%28by%20ref%29))
  (bind (_instigator _targetdata) (Utilities|Struct|BreakGameplayEventData EventData))
  (|CallParentFunction EventData)
  (Utilities|FlowControl|Sequence
    (:then_0
      (bind _asfort_player_pawn (Utilities|Casting|CastToFortPlayerPawn _instigator))
      (Variables|Default|SetInstigator _asfort_player_pawn)
      (if (Ability|TargetData|TargetDataHasHitResult _targetdata)
        (bind _asb_wild_estate_cove_spice (Utilities|Casting|CastToB_WildEstate_CoveSpice (Weapon|GetCurrentWeapon (Variables|Default|GetInstigator))))
        (Class|BWildEstateCoveSpice|UpdateBeamAudioHitResult _asb_wild_estate_cove_spice (Ability|TargetData|GetHitResultfromTargetData _targetdata))))
    (:then_1
      (Ability|EndAbility))))
