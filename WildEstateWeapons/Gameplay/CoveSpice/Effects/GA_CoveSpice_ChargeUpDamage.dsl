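; Gameplay ability for the CoveSpice weapon that applies the ChargeUpDamage gameplay effect when triggered by a gameplay event.
; On activation it calls the parent handler, then casts the event's Target to FortPawn.
; If the cast succeeds, the effect goes on the ability owner, not the target; the pawn cast serves only as a filter.
; The ability then ends immediately.
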(event Ability|EventActivateAbilityFromEvent
  (params (EventData :type Gameplay%20Event%20Data%20Structure%20%28by%20ref%29))
  (bind (_eventtag _instigator _target) (Utilities|Struct|BreakGameplayEventData EventData))
  (|CallParentFunction EventData)
  (Utilities|Casting|CastToFortPawn _target
    (:then
      (Ability|ApplyGameplayEffectToOwner "/WildEstateWeapons/Gameplay/CoveSpice/Effects/GE_CoveSpice_ChargeUpDamage.GE_CoveSpice_ChargeUpDamage_C"))
    (:CastFailed
      _))
  (Ability|EndAbility))
