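; CoveSpice weapon gameplay ability, triggered by a gameplay event, that reports damage charge as a percentage on a pawn.
; On first run it adds the ChargePercent component to the event's pawn and applies the Charge Percent Tag effect.
; Each activation passes the avatar's Damage Charge effect count, safely divided by a class default of that effect, to the component's UpdatePercent, then ends.
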(event Ability|EventActivateAbilityFromEvent
  (params (EventData :type Gameplay%20Event%20Data%20Structure%20%28by%20ref%29))
  (|CallParentFunction EventData)
  (bind _asfort_pawn (Utilities|Casting|CastToFortPawn (Utilities|Struct|BreakGameplayEventData EventData))
    (:then
      (Utilities|FlowControl|Sequence
        (:then_0
          (Utilities|IsValid (Actor|GetComponentbyClass _asfort_pawn "/WildEstateWeapons/Gameplay/CoveSpice/AC_CoveSpice_ChargePercent.AC_CoveSpice_ChargePercent_C")
            (:Is%20Not%20Valid
              (Game|AddComponentbyClass _asfort_pawn "/WildEstateWeapons/Gameplay/CoveSpice/AC_CoveSpice_ChargePercent.AC_CoveSpice_ChargePercent_C")
              (Ability|ApplyGameplayEffectToTarget (Ability|TargetData|AbilityTargetDatafromActor _asfort_pawn) (Variables|Default|GetCharge%20Percent%20Tag%20GE)))))
        (:then_1
          (Class|ACCoveSpiceChargePercent|UpdatePercent (Actor|GetComponentbyClass _asfort_pawn "/WildEstateWeapons/Gameplay/CoveSpice/AC_CoveSpice_ChargePercent.AC_CoveSpice_ChargePercent_C") (CallFunction|GetPawnChargePercent))
          (Ability|EndAbility))))
    (:CastFailed
      (Ability|EndAbility))))

(fn GetPawnChargePercent
  (return (Math|Float|SafeDivide (Math|Conversions|ToFloat(Integer) (GameplayEffects|GetGameplayEffectCount (Ability|GetFortAbilitySystemComponent (CallFunction|GetAvatarasFortPlayerPawn)) (Variables|Default|GetDamage%20Charge%20GE) 0 true)) (Math|Conversions|ToFloat(Integer) (Class|GetClassDefaults (Variables|Default|GetDamage%20Charge%20GE))))))
