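; Base passive ability for the CoveSpice weapon that updates a HUD reticle based on whether a target is in range.
; On activation it caches the player pawn and controller, then starts a looping timer, cleared on ability end.
; Each tick runs the TargetInRange targeting preset and calls OnUpdateReticle with whether any actors were found.
; CheckTargetInRange, an alternative distance, line-trace and damageability check, is not called here.
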(event Ability|EventActivateAbility
  (|CallParentFunction)
  ;[0 0.15 0.01338 0.5] Setup Variables
  (bind _outfortplayerpawn (CallFunction|GetAvatarasValidFortPlayerPawn)
    (:ValidFortPlayerPawn
      (bind _output_get (Variables|Default|SetPlayer%20Pawn _outfortplayerpawn))
      (Utilities|IsValid (Fortnite|GetFortPlayerControllerfromActor _output_get)
        (:Is%20Valid
          (Variables|Default|SetPlayer%20Controller (Fortnite|GetFortPlayerControllerfromActor _output_get))
          ;[0.146 0 0.15 0.5] Start Reticle Timer Update
          (if (Utilities|Time|IsValidTimerHandle (Variables|Default|GetReticle%20Update%20Timer))
            (Utilities|Time|ClearandInvalidateTimerbyHandle (Variables|Default|GetReticle%20Update%20Timer))
            (else
              _))
          (bind _returnvalue (Utilities|Time|SetTimerbyEvent (EventDispatchers|CreateEvent :!event Custom|Update%20Reticle) (ScalableFloat|GetValueatLevel (Variables|Default|GetReticle%20Update%20Tick)) true true))
          (Variables|Default|SetReticle%20Update%20Timer _returnvalue))
        (:Is%20Not%20Valid
          _)))
    (:AvatarCastFailed
      _))
  (Ability|EndAbility))

(event Ability|EventOnEndAbility
  (params (bWasCancelled :type Boolean))
  (|CallParentFunction bWasCancelled)
  (Utilities|Time|ClearandInvalidateTimerbyHandle (Variables|Default|GetReticle%20Update%20Timer)))

(event Custom|Update%20Reticle
  ;[0.15 0.1248 0 0.5] Update Reticle
  (bind (_asynctaskproxy _targetdata) (Ability|Tasks|PerformTargetingAbilityTask "None" "/WildEstateWeapons/Gameplay/CoveSpice/TP_CoveSpice_TargetInRange.TP_CoveSpice_TargetInRange" false false true)
    (:Targeted
      (CallFunction|CheckForValidTarget :Actors (Ability|TargetData|GetActorsfromTargetData _targetdata)))))

(fn Checkforvalidtarget
  (params (TargetingRequestHandle :type Targeting%20Request%20Handle%20Structure) (Actors :type Array%20of%20Actor%20Object%20References))
  (Default|CallOnUpdateReticle (> (Utilities|Array|Length Actors) 0)))

(fn CheckTargetInRange
  (Utilities|FlowControl|Sequence
    (:then_0
      (LocalVariables|SetActor%20Location (Transformation|GetActorLocation (Variables|Default|GetPlayer%20Pawn)))
      (LocalVariables|SetTarget%20Location (Reticle|GetLocationUnderReticle (Variables|Default|GetPlayer%20Controller)))
      (if (<= (Math|Vector|Distance(Vector) (LocalVariables|GetActor%20Location) (LocalVariables|GetTarget%20Location)) (ScalableFloat|GetValueatLevel (Variables|Default|GetWeapon%20Target%20Range)))
        (bind _outhit (Collision|LineTraceByChannel (LocalVariables|GetActor%20Location) (LocalVariables|GetTarget%20Location) "TraceTypeQuery8" false (Utilities|Array|MakeArray (Variables|Default|GetPlayer%20Pawn)) "ForOneFrame"))
        (if (Class|BPFLAICommonFunctions|GetCanbeDamaged (.HitActor (Collision|BreakHitResult _outhit)))
          (LocalVariables|SetTarget%20Location%20In%20Range (.CanBeDamaged (Class|BPFLAICommonFunctions|GetCanbeDamaged (.HitActor (Collision|BreakHitResult _outhit)))))
          (else
            (LocalVariables|SetTarget%20Location%20In%20Range false)))
        (else
          (LocalVariables|SetTarget%20Location%20In%20Range false))))
    (:then_1
      (Default|CallOnUpdateReticle (LocalVariables|GetTarget%20Location%20In%20Range)))))
