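; Base gameplay ability for the Wild Estate Cove Spice effects, named as the parent for that effect family.
; On activation it calls the parent implementation, then casts the avatar actor to PlayerPawn_Athena and stores the result in the Target Pawn variable.
; It applies no effect itself in this graph and only caches the pawn reference.
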(event Ability|EventActivateAbility
  (|CallParentFunction)
  (bind _asplayer_pawn_athena (Utilities|Casting|CastToPlayerPawn_Athena (Ability|GetAvatarActorfromActorInfo)))
  (Variables|Default|SetTarget%20Pawn _asplayer_pawn_athena))
