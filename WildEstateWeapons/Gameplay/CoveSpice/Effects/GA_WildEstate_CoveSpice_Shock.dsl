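; Gameplay ability for the WildEstate CoveSpice weapon's shock effect, extending a parent ability.
; On activation it runs the parent logic, then attaches a random hat to the target pawn's "Hat" socket if target and instigator are friendly.
; Helpers limit chain shocks to non-friendly players other than the original instigator and target, and time ability end to the speed-scaled shock animation.
; Ending the ability clears the color change timer.
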(event Ability|EventActivateAbility
  (|CallParentFunction)
  (Variables|Default|GetTarget%20Pawn
    (:then
      (switch Utilities|FlowControl|Switch|SwitchonEFortTeamAffiliation (Team|GetTeamAffiliation (Variables|Default|GetTarget%20Pawn) (Ability|EffectContext|GetInstigatorActor (Ability|GetGrantedbyEffectContext)))
        (:Friendly
          (CallFunction|ApplyRandomHat))))
    (:else)))

(event Ability|EventOnEndAbility
  (params (bWasCancelled :type Boolean))
  (|CallParentFunction bWasCancelled)
  (Utilities|Time|ClearandInvalidateTimerbyHandle (Variables|Default|GetColor%20Change%20Timer)))

(event Custom|Once%20Per%20Shot%20End%20Ability
  (Utilities|FlowControl|Delay (CallFunction|GetShockAnimPlayLength)) ; wait for anim to finish playing and then end ability
  (Ability|EndAbility))

(event Custom|Apply%20Random%20Hat
  (Attachments|AttachAssettoSocket (Variables|Default|GetTarget%20Pawn) (Utilities|Struct|MakeAssetAttachment "Hat" (Utilities|Array|RandomArrayItem (Variables|Default|GetRandom%20Hats)) 0 false)))

(fn CanApplyChainShockTo
  (params (Player%20Pawn :type Fort%20Player%20Pawn%20Object%20Reference))
  (if (Math|Boolean|ANDBoolean (!= (Variables|/Script/Engine.Pawn|GetPlayerState Player%20Pawn) (Ability|EffectContext|GetOriginalInstigatorActor (Ability|GetGrantedbyEffectContext))) (!= (Variables|/Script/Engine.Pawn|GetPlayerState Player%20Pawn) (Variables|/Script/Engine.Pawn|GetPlayerState (Variables|Default|GetTarget%20Pawn))) (not (Utilities|Enum|Equal(Enum) (Team|GetTeamAffiliation (Ability|EffectContext|GetOriginalInstigatorActor (Ability|GetGrantedbyEffectContext)) Player%20Pawn) "Friendly")))
    (return true)
    (else
      (return false))))

(fn GetShockAnimPlayLength
  (return (/ (Animation|GetPlayLength (Variables|Default|GetShock%20Anim)) (ScalableFloat|GetValueatLevel (Variables|Default|GetShock%20Speed%20Multiplier)))))
