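; Gameplay effect custom application requirement that gates the CoveSpice Shock effect by team affiliation.
; CanApplyGameplayEffect casts the ASC owner to FortPlayerState and gets its current pawn.
; It compares that pawn's team with the original instigator from the spec's effect context, allowing application to neutral or hostile targets and blocking friendly ones.
; A failed cast or invalid pawn is intended to block application.
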(fn CanApplyGameplayEffect
  (params (GameplayEffect :type Gameplay%20Effect%20Object%20Reference) (Spec :type Gameplay%20Effect%20Spec%20Structure%20%28by%20ref%29) (ASC :type Ability%20System%20Component%20Object%20Reference))
  ;[1 1 1] Checking for team affiliation. Can apply is neutral or enemy, and cannot if target firndly or invalid.
  (Utilities|FlowControl|Sequence
    (:then_0
      (bind _asfort_player_state (Utilities|Casting|CastToFortPlayerState (Components|GetOwner ASC))
        (:then
          (Utilities|IsValid (PlayerState|GetCurrentPawn _asfort_player_state)
            (:Is%20Valid
              (LocalVariables|SetCanApply (Utilities|Select :Index (Team|GetTeamAffiliation (PlayerState|GetCurrentPawn _asfort_player_state) (Ability|EffectContext|GetOriginalInstigatorActor (Ability|GameplayEffect|GetEffectContextfromSpec Spec))) :Friendly false :Neutral true :Hostile true)) ; can apply to neutral or hostile
            )
            (:Is%20Not%20Valid
              _)))
        (:CastFailed
          _))
      (LocalVariables|SetCanApply false) ; false if player or target player is invalid
    )
    (:then_1
      (return (LocalVariables|GetCanApply)))))
