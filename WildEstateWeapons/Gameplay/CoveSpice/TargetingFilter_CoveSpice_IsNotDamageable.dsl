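; Targeting filter for the CoveSpice weapon that drops targets considered not damageable, judged by immunity gameplay tags.
; The ShouldRemoveTarget function breaks the target's hit result and asks the hit actor, through an interface message, whether it has any of the Immunity Tags.
; A match removes the target, as does an invalid targeting subsystem for the request handle.
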(fn ShouldRemoveTarget
  (params (TargetingHandle :type Targeting%20Request%20Handle%20Structure%20%28by%20ref%29) (TargetData :type Targeting%20Default%20Result%20Data%20Structure%20%28by%20ref%29))
  (bind (_bblockinghit _binitialoverlap _time _distance _location _impactpoint _normal _impactnormal _physmat _hitactor) (Collision|BreakHitResult (Utilities|Struct|BreakTargetingDefaultResultData TargetData)))
  (Utilities|IsValid (Targeting|GetTargetingSubsystem TargetingHandle)
    (:Is%20Valid
      (bind _returnvalue (GameplayTags|HasAnyMatchingGameplayTags(Message) _hitactor (GameplayTags|MakeGameplayTagContainerfromArray (Variables|Default|GetImmunity%20Tags))))
      (return _returnvalue))
    (:Is%20Not%20Valid
      (return true))))
